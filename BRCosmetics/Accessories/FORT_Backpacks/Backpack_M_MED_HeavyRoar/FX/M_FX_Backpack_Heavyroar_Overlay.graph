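# Translucent unlit Niagara mesh particle material for the overlay effect on the HeavyRoar backpack.
# Two panning noise samples multiply to vary particle color between dimmed and full.
# A packed overlay texture supplies the main mask in R and shadow in G, driving color and opacity with particle alpha.
# Its B channel feeds smoothsteps offset by the FlashLetters dynamic parameter, creating a moving band that brightens letter highlights.
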Material /BRCosmetics/Accessories/FORT_Backpacks/Backpack_M_MED_HeavyRoar/FX/M_FX_Backpack_Heavyroar_Overlay
graph {
 properties {
  bUsedWithNiagaraMeshParticles=true
  blendMode=BLEND_Translucent
  pixelDepthOffsetMode=PDOM_Legacy
  shadingModel=MSM_Unlit
  usedShadingModels=Unlit
 }
 nodes [
  Constant()
  n1 = ParticleColor()
  n2 = Multiply(A=@n1.RGB, B=ScalarParameter() {defaultValue=0.3, parameterName=Color 2 Multiplier})
  n3 = TextureCoordinate()
  n4 = VectorParameter() {parameterName=Noise UV 1 Scale Speed, defaultValue=<rgba 0.5 0.2 -0.05 -0.07>}
  n5 = MaterialFunctionCall(X=@n4.R, Y=@n4.R) {materialFunction=&MakeFloat2}
  n6 = TextureObject() {samplerType=SAMPLERTYPE_Grayscale, texture=/FortUILibrary/UIKit/Buttons/Textures/LowResBlurredNoise}
  n7 = TextureSample(
   UVs=Panner(
    Coordinate=Multiply(A=@n3, B=@n5.Result)
    Speed=MaterialFunctionCall(X=@n4.B, Y=@n4.B) {materialFunction=&MakeFloat2}.Result
   )
   Tex=@n6
  ) {samplerType=SAMPLERTYPE_Grayscale, texture=&LowResBlurredNoise}
  n8 = TextureCoordinate()
  n9 = VectorParameter() {parameterName=Noise UV 2 Scale Speed, defaultValue=<rgba 0.5 0.5 -0.025 0.05>}
  n10 = MaterialFunctionCall(X=@n9.R, Y=@n9.R) {materialFunction=&MakeFloat2}
  n11 = TextureSample(
   UVs=Panner(
    Coordinate=Multiply(A=@n8, B=@n10.Result)
    Speed=MaterialFunctionCall(X=@n9.B, Y=@n9.B) {materialFunction=&MakeFloat2}.Result
   )
   Tex=@n6
  ) {samplerType=SAMPLERTYPE_Grayscale, texture=&LowResBlurredNoise}
  n12 = Saturate(None=Multiply(A=Multiply(A=@n7.R, B=@n11.R)) {constB=2})
  n13 = ScalarParameter() {defaultValue=6, parameterName=LetterHighlightColor}
  n14 = NamedRerouteUsage()
  n15 = TextureSample() {texture=&/T_FX_Backpack_HeavyRoar_Overlay}
  LinearInterpolate(
   A=VectorParameter() {parameterName=Shadow Color, defaultValue=<rgba 0 0 0 1>}.RGB
   B=Multiply(A=LinearInterpolate(A=@n2, B=@n1.RGB, Alpha=@n12), B=LinearInterpolate(B=@n13, Alpha=@n14) {constA=1})
   Alpha=@n15.R
  )
  n16 = Multiply(A=@n15.G, B=ScalarParameter() {defaultValue=0.5, parameterName=Shadow Opacity})
  Multiply(A=Saturate(None=Add(A=@n15.R, B=@n16)), B=ParticleColor().A)
  n17 = DynamicParameter() {defaultValue=<rgba 0 1 1 1>, paramNames [FlashLetters,Param2,Param3,Param4]}
  n18 = Add(A=@n17.FlashLetters, B=ScalarParameter())
  n19 = Multiply(A=@n15.B)
  n20 = Add(A=@n17.FlashLetters) {constB=0.01}
  n21 = ScalarParameter()
  n22 = OneMinus(None=SmoothStep(Min=@n20, Max=Add(A=@n20, B=@n21), Value=@n19))
  NamedRerouteDeclaration(None=Multiply(A=SmoothStep(Min=@n17.FlashLetters, Max=@n18, Value=@n19), B=Multiply(A=@n22))) {name=LetterHighlight, nodeColor=<rgba 0 0.25 1 1>}
  ScalarParameter() {defaultValue=0.91, parameterName=s}
 ]
}
