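# Niagara system for the HeavyRoar backpack cosmetic: audio-reactive discs, flickering orange lights, a glow overlay and socket-attached smoke.
# Left and right layered disc meshes blend idle sine waves into reshaped event audio spectrum by an audio gate, feeding material parameters.
# The same gate fades in the lights, while curves flicker lights and overlay.
# Smoke meshes pass lagged positions to their material; main smoke intensifies with KillReactBlend and adjusts for skydiving.
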NiagaraSystem /BRCosmetics/Accessories/FORT_Backpacks/Backpack_M_MED_HeavyRoar/FX/NS_Backpack_HavyRoar
dependencies {
 usedDataInterfaces [
  &NiagaraDataInterfaceArrayInt32
  &NiagaraDataInterfaceCurve
  &NiagaraDataInterfaceArrayFloat3
  &NiagaraDataInterfaceArrayFloat
  &NiagaraDataInterfaceSkeletalMesh
 ]
 usedDynamicInputs [&ClampFloat]
 usedModules [
  &self:SystemSpawnScript.SystemScriptSource.SystemScriptGraph.NiagaraNodeAssignment_0.SetVariables_#1
  &self:SystemSpawnScript.SystemScriptSource.SystemScriptGraph.NiagaraNodeAssignment_2.SetVariables_#2
  &SystemState
  &self:SystemSpawnScript.SystemScriptSource.SystemScriptGraph.NiagaraNodeAssignment_1.SetVariables_#3
 ]
}
emitters {
 Light {
  emitterUpdateScript [
   &EmitterState(Inactive Response=::NewEnumerator1, Life Cycle Mode=::NewEnumerator1, Loop Duration=2, UseLoopDelay=true)
   &SpawnBurst_Instantaneous(Spawn Count=2, Use Spawn Probability=false)
  ]
  particleSpawnScript [
   &FN_SetSpawnGroup(Number of Spawn Groups=2)
   &InitializeParticle(
    Color Mode=::NewEnumerator0
    Lifetime=4
    Position Offset=<20 -20 0>
    Sprite Size Mode=::NewEnumerator3
    Uniform Sprite Size=5
    UsePositionOffset=true
    Write Material Random=true
   ) {module=InitializeParticle001}
   &SystemLocation(
    MaskSpawnGroup=false
    Offset=&SelectVectorFromArray(Direct Array Index=$Particles.SpawnGroup, Vector Selection Array=DataInterface() {propertyValues {FloatData [<11 -15 -7>, <-9.8 -17 -7>]}})
   )
   &SystemLocation(MaskSpawnGroup=false, Offset=<0 -20 -8>) {enabled=false, module=SystemLocation001}
  ]
  particleUpdateScript [
   &ParticleState()
   &Color(Color=&MakeLinearColorFromVectorAndFloat(
     "Vector (RGB)"=&Multiply_Vector(
      A=&MakeVectorFromLinearColorRGB(Color=&LerpLinearColors(
        StartColor=<rgba 0 0 0 0>
        EndColor=&LerpLinearColors(
         StartColor=<rgba 1 0.5607 0.0237 1>
         EndColor=<rgba 1.25 0.5033 0.0391 1>
         LerpFactor=&FloatFromCurve(
          FloatCurve=DataInterface() {propertyValues {Curve {keys [<at 0 1>,<at 0.1069 0.0895>,<at 0.3206 0.7423>,<at 0.4907 0.1444>,<at 0.6992 0.8948>,<at 0.9061 0.1261>,<at 1 1>]}}}
          CurveIndex=&Multiply_Float(A=$Emitter.NormalizedLoopAge, B=1)
         )
        )
        LerpFactor=$System.IncomingAudioGate
       ))
      B=&VectorFromFloat(Value=&FloatFromCurve(
        FloatCurve=DataInterface() {propertyValues {Curve {keys [<at 0 1>,<at 0.0584 1.1216>,<at 0.0813 0.573>,<at 0.2 0.2>,<at 0.222 1.0327>,<at 0.2722 0.5884>,<at 0.3183 0.7837>,<at 0.4 0.25>,<at 0.4424 0.5444>,<at 0.4815 1.1358>,<at 0.5388 1.0006>,<at 0.6 0.2>,<at 0.6496 0.7446>,<at 0.6954 0.6498>,<at 0.737 1.0896>,<at 0.8 0.25>,<at 0.8926 0.7771>,<at 0.9466 0.5225>,<at 1 1>]}}}
        CurveIndex=$Emitter.NormalizedLoopAge
       ))
     )
     "Float (Alpha)"=1
    ))
   &FN_Light_v2(Light Exponent=1, Light Radius=17, LightEnabled=true, Scale Color=1)
  ]
  rendererClasses [&NiagaraLightRendererProperties]
  renderers [{rendererClass=&NiagaraLightRendererProperties, rendererIndex=0}]
 }
 OuterDisc_L {
  emitterUpdateScript [
   &EmitterState(
    Life Cycle Mode=::NewEnumerator1
    Loop Behavior=::NewEnumerator1
    Loop Duration Mode=::NewEnumerator1
    Scalability Mode=::NewEnumerator1
    UseLoopDelay=false
   )
   &SpawnBurst_Instantaneous(Use Spawn Probability=false, UseLoopCountLimit=false)
   /SetVariables(Emitter.IncomingSpectrumData:LinearColor=&LerpLinearColors(
     StartColor=&MakeLinearColor(
      R:NiagaraFloat=&Sine(Normalized Angle=$Emitter.Age, Period=4, Scale=0.1, Bias=0.4)
      G:NiagaraFloat=&Cosine(Normalized Angle=$Emitter.Age, Period=2.8, Scale=0.1, Bias=0.4)
      B:NiagaraFloat=&Sine(Normalized Angle=$Emitter.Age, Period=3.25, Scale=0.1, Bias=0.3)
      A:NiagaraFloat=&Cosine(Normalized Angle=$Emitter.Age, Period=4, Scale=0.1, Bias=0.4)
     )
     EndColor=/Niagara/DynamicInputs/LinearColor/MakeLinearColorFromFloat4(Vector:Vector4f=&MakeVector4(
       X=&OneMinusFloat(Float=&Power(
         a=&OneMinusFloat(Float=&MakeFloatFromLinearColor(LinearColor=$NPC.FortniteNPC.EventAudio_Spectrum, Channel:ENiagaraLinearColor_Channels=::NewEnumerator0))
         b=0.5
        ))
       Y=&OneMinusFloat(Float=&Power(
         a=&OneMinusFloat(Float=&MakeFloatFromLinearColor(LinearColor=$NPC.FortniteNPC.EventAudio_Spectrum, Channel:ENiagaraLinearColor_Channels=::NewEnumerator1))
         b=0.75
        ))
       Z=&OneMinusFloat(Float=&Power(
         a=&OneMinusFloat(Float=&MakeFloatFromLinearColor(LinearColor=$NPC.FortniteNPC.EventAudio_Spectrum, Channel:ENiagaraLinearColor_Channels=::NewEnumerator2))
         b=1
        ))
       W=&OneMinusFloat(Float=&Power(
         a=&OneMinusFloat(Float=&MakeFloatFromLinearColor(LinearColor=$NPC.FortniteNPC.EventAudio_Spectrum, Channel:ENiagaraLinearColor_Channels=::NewEnumerator3))
         b=1.2
        ))
      ))
     LerpFactor=$System.IncomingAudioGate
    )) {isSetParameters=true, module=SetVariables_#4}
  ]
  particleSpawnScript [
   &InitializeParticle(Color=<rgba 1 1 1 0.9>, Lifetime=5, Mesh Scale Mode=::NewEnumerator3, UsePositionOffset=false, Write Material Random=true)
   /SetVariables(Particles.LongerConeColor:LinearColor=<rgba 1 1 1 0.1>) {isSetParameters=true, module=SetVariables_#5}
   /SetVariables(
    Particles.ScaleBase:Vector3f=&VectorFromFloat(Value=1.08)
    Particles.ScaleInner_01:Vector3f=&VectorFromFloat(Value=0.67)
    Particles.ScaleInner_02:Vector3f=&VectorFromFloat(Value=0.3)
   ) {isSetParameters=true, module=SetVariables_#6}
  ]
  particleUpdateScript [
   &ParticleState(Kill Particles When Lifetime Has Elapsed=false)
   /SetVariables(Particles.SpectrumSample:NiagaraFloat=&Add_Float(
     A=&self:MaxOfColorChannels(Color:LinearColor=&MultiplyLinearColor(Linear Color=$Emitter.IncomingSpectrumData, Scale Factor=<1 1 1 1>))
     B=0
    )) {isSetParameters=true, module=SetVariables_#7}
   &DynamicMaterialParameters(
    Index 0 Param 1=HlslExpression() {hlslExpression=Emitter.IncomingSpectrumData.x}
    Index 0 Param 2=HlslExpression() {hlslExpression=Emitter.IncomingSpectrumData.y}
    Index 0 Param 3=HlslExpression() {hlslExpression=Emitter.IncomingSpectrumData.z}
    Index 0 Param 4=HlslExpression() {hlslExpression=Emitter.IncomingSpectrumData.a}
    Index 1 Param 1=$System.InterpolateOverTime.InterpolateOverTime_Float.Moving Average
    Module.Param0WriteEnabled=true
    Module.Param0WriteEnabled1=true
    Module.Param1WriteEnabled=true
    Module.Param1WriteEnabled1=true
    Module.Param2WriteEnabled=true
    Module.Param2WriteEnabled1=true
    Module.Param3WriteEnabled=true
    Module.Param3WriteEnabled1=true
    Write Parameter Index 1=true
   )
   &SkeletalMeshLocation(
    MaskSpawnGroup=false
    Mesh Sampling Type=::NewEnumerator5
    Orientation Sampling=::NewEnumerator1
    Position Sampling=::NewEnumerator2
    Skeletal Mesh=$System.Skeletal Mesh
    Socket Sampling Mode=::NewEnumerator5
   )
   /SetVariables(Particles.MeshOrientation=$Particles.SkeletalMeshLocation.SampledRotation) {isSetParameters=true, module=SetVariables_#8}
   /SetVariables(
    Particles.BasePosition:NiagaraPosition=&AddVectorToPosition(
     Position=$Particles.SkeletalMeshLocation.SampledPosition
     Vector=&Multiply_VectorByFloat(
      Vector=&ReturnMeshOrientationAxis(Mesh Orientation Axis=::ZAxis, Mesh Orientation Quat=$Particles.MeshOrientation)
      Float=-0.27
     )
    )
    Particles.InnerPosition_01:NiagaraPosition=&AddVectorToPosition(
     Position=$Particles.SkeletalMeshLocation.SampledPosition
     Vector=&Multiply_VectorByFloat(
      Vector=&ReturnMeshOrientationAxis(Mesh Orientation Axis=::ZAxis, Mesh Orientation Quat=$Particles.MeshOrientation)
      Float=0.3
     )
    )
    Particles.InnerPosition_02:NiagaraPosition=&AddVectorToPosition(
     Position=$Particles.SkeletalMeshLocation.SampledPosition
     Vector=&Multiply_VectorByFloat(
      Vector=&ReturnMeshOrientationAxis(Mesh Orientation Axis=::ZAxis, Mesh Orientation Quat=$Particles.MeshOrientation)
      Float=-0.05
     )
    )
   ) {isSetParameters=true, module=SetVariables_#9}
  ]
  rendererClasses [&NiagaraMeshRendererProperties]
  renderers [
   {rendererClass=&NiagaraMeshRendererProperties, rendererIndex=0}
   {rendererClass=&NiagaraMeshRendererProperties, rendererIndex=1}
   {rendererClass=&NiagaraMeshRendererProperties, rendererIndex=2}
   {rendererClass=&NiagaraMeshRendererProperties, rendererIndex=3}
  ]
 }
 OuterDisc_R {
  emitterUpdateScript [
   &EmitterState(
    Life Cycle Mode=::NewEnumerator1
    Loop Behavior=::NewEnumerator1
    Loop Duration Mode=::NewEnumerator1
    Scalability Mode=::NewEnumerator1
    UseLoopDelay=false
   )
   &SpawnBurst_Instantaneous(Use Spawn Probability=false, UseLoopCountLimit=false)
   /SetVariables(Emitter.IncomingSpectrumData:LinearColor=&LerpLinearColors(
     StartColor=&MakeLinearColor(
      R:NiagaraFloat=&Sine(Normalized Angle=$Emitter.Age, Period=4, Scale=0.1, Bias=0.4)
      G:NiagaraFloat=&Cosine(Normalized Angle=$Emitter.Age, Period=2.8, Scale=0.1, Bias=0.4)
      B:NiagaraFloat=&Sine(Normalized Angle=$Emitter.Age, Period=3.25, Scale=0.1, Bias=0.3)
      A:NiagaraFloat=&Cosine(Normalized Angle=$Emitter.Age, Period=4, Scale=0.1, Bias=0.4)
     )
     EndColor=&MakeLinearColorFromFloat4(Vector:Vector4f=&MakeVector4(
       X=&OneMinusFloat(Float=&Power(
         a=&OneMinusFloat(Float=&MakeFloatFromLinearColor(LinearColor=$NPC.FortniteNPC.EventAudio_Spectrum, Channel:ENiagaraLinearColor_Channels=::NewEnumerator0))
         b=0.5
        ))
       Y=&OneMinusFloat(Float=&Power(
         a=&OneMinusFloat(Float=&MakeFloatFromLinearColor(LinearColor=$NPC.FortniteNPC.EventAudio_Spectrum, Channel:ENiagaraLinearColor_Channels=::NewEnumerator1))
         b=0.75
        ))
       Z=&OneMinusFloat(Float=&Power(
         a=&OneMinusFloat(Float=&MakeFloatFromLinearColor(LinearColor=$NPC.FortniteNPC.EventAudio_Spectrum, Channel:ENiagaraLinearColor_Channels=::NewEnumerator2))
         b=1
        ))
       W=&OneMinusFloat(Float=&Power(
         a=&OneMinusFloat(Float=&MakeFloatFromLinearColor(LinearColor=$NPC.FortniteNPC.EventAudio_Spectrum, Channel:ENiagaraLinearColor_Channels=::NewEnumerator3))
         b=1.2
        ))
      ))
     LerpFactor=$System.IncomingAudioGate
    )) {isSetParameters=true, module=SetVariables_#4}
  ]
  particleSpawnScript [
   &InitializeParticle(Color=<rgba 1 1 1 0.9>, Lifetime=5, Mesh Scale Mode=::NewEnumerator3, UsePositionOffset=false, Write Material Random=true)
   /SetVariables(Particles.LongerConeColor:LinearColor=<rgba 1 1 1 0.05>) {isSetParameters=true, module=SetVariables_#5}
   /SetVariables(
    Particles.ScaleBase:Vector3f=&VectorFromFloat(Value=1.08)
    Particles.ScaleInner_01:Vector3f=&VectorFromFloat(Value=0.67)
    Particles.ScaleInner_02:Vector3f=&VectorFromFloat(Value=0.3)
   ) {isSetParameters=true, module=SetVariables_#6}
  ]
  particleUpdateScript [
   &ParticleState(Kill Particles When Lifetime Has Elapsed=false)
   /SetVariables(Particles.SpectrumSample:NiagaraFloat=&Add_Float(
     A=&self:MaxOfColorChannels(Color:LinearColor=&MultiplyLinearColor(Linear Color=$Emitter.IncomingSpectrumData, Scale Factor=<1 1 1 1>))
     B=0
    )) {isSetParameters=true, module=SetVariables_#7}
   &DynamicMaterialParameters(
    Index 0 Param 1=HlslExpression() {hlslExpression=Emitter.IncomingSpectrumData.x}
    Index 0 Param 2=HlslExpression() {hlslExpression=Emitter.IncomingSpectrumData.y}
    Index 0 Param 3=HlslExpression() {hlslExpression=Emitter.IncomingSpectrumData.z}
    Index 0 Param 4=HlslExpression() {hlslExpression=Emitter.IncomingSpectrumData.a}
    Index 1 Param 1=$System.InterpolateOverTime.InterpolateOverTime_Float.Moving Average
    Module.Param0WriteEnabled=true
    Module.Param0WriteEnabled1=true
    Module.Param1WriteEnabled=true
    Module.Param1WriteEnabled1=true
    Module.Param2WriteEnabled=true
    Module.Param2WriteEnabled1=true
    Module.Param3WriteEnabled=true
    Module.Param3WriteEnabled1=true
    Write Parameter Index 1=true
   )
   &SkeletalMeshLocation(
    Bone / Socket Index=1
    MaskSpawnGroup=false
    Mesh Sampling Type=::NewEnumerator5
    Orientation Sampling=::NewEnumerator1
    Skeletal Mesh=$System.Skeletal Mesh
    Socket Sampling Mode=::NewEnumerator5
   )
   /SetVariables(Particles.MeshOrientation=&MultiplyQuaternion(
     Quaternion A=$Particles.SkeletalMeshLocation.SampledRotation
     Quaternion B=&MakeQuatFromAxisAngle(Axis=<0 0 1>, AngleInDegrees=180)
    )) {isSetParameters=true, module=SetVariables_#8}
   /SetVariables(
    Particles.BasePosition:NiagaraPosition=&AddVectorToPosition(
     Position=$Particles.SkeletalMeshLocation.SampledPosition
     Vector=&Multiply_VectorByFloat(
      Vector=&ReturnMeshOrientationAxis(Mesh Orientation Axis=::ZAxis, Mesh Orientation Quat=$Particles.MeshOrientation)
      Float=-0.27
     )
    )
    Particles.InnerPosition_01:NiagaraPosition=&AddVectorToPosition(
     Position=$Particles.SkeletalMeshLocation.SampledPosition
     Vector=&Multiply_VectorByFloat(
      Vector=&ReturnMeshOrientationAxis(Mesh Orientation Axis=::ZAxis, Mesh Orientation Quat=$Particles.MeshOrientation)
      Float=0.3
     )
    )
    Particles.InnerPosition_02:NiagaraPosition=&AddVectorToPosition(
     Position=$Particles.SkeletalMeshLocation.SampledPosition
     Vector=&Multiply_VectorByFloat(
      Vector=&ReturnMeshOrientationAxis(Mesh Orientation Axis=::ZAxis, Mesh Orientation Quat=$Particles.MeshOrientation)
      Float=-0.05
     )
    )
   ) {isSetParameters=true, module=SetVariables_#9}
  ]
  rendererClasses [&NiagaraMeshRendererProperties]
  renderers [
   {rendererClass=&NiagaraMeshRendererProperties, rendererIndex=0}
   {rendererClass=&NiagaraMeshRendererProperties, rendererIndex=1}
   {rendererClass=&NiagaraMeshRendererProperties, rendererIndex=2}
   {rendererClass=&NiagaraMeshRendererProperties, rendererIndex=3}
  ]
 }
 Slimshady_Overlay {
  emitterUpdateScript [
   &EmitterState(Life Cycle Mode=::NewEnumerator1, Loop Behavior=::NewEnumerator1, Loop Duration=1, Scalability Mode=::NewEnumerator1, UseLoopDelay=false)
   &SpawnBurst_Instantaneous(Use Spawn Probability=false, UseLoopCountLimit=false)
  ]
  particleSpawnScript [&InitializeParticle(Color=<rgba 10 4.1667 0 1>, Lifetime=5, Mesh Scale Mode=::NewEnumerator3, UsePositionOffset=false, Write Material Random=true)]
  particleUpdateScript [
   &ParticleState(Kill Particles When Lifetime Has Elapsed=false)
   &ScaleColor(
    Scale Alpha=$System.InterpolateOverTime.InterpolateOverTime_Float.Moving Average
    Scale RGB=&VectorFromFloat(Value=&Multiply_Float(
      A=&Lerp_Float(
       A=6
       B=0.8
       Alpha=&FloatFromCurve(
        FloatCurve=DataInterface() {propertyValues {Curve {keys [<at 0 0.85>,<at 0.0813 0.573>,<at 0.1 1>,<at 0.1868 0.0837>,<at 0.222 1.0327>,<at 0.2722 0.5884>,<at 0.3183 0.7837>,<at 0.4424 0.5444>,<at 0.4815 1>,<at 0.4979 0.0387>,<at 0.5388 1.0006>,<at 0.6033 0.0575>,<at 0.6496 0.7446>,<at 0.6954 0.6498>,<at 0.737 1.0896>,<at 0.8 0.25>,<at 0.8926 0.7771>,<at 0.9466 0.5225>,<at 1 1>]}}}
        CurveIndex=&Multiply_Float(A=$Emitter.NormalizedLoopAge, B=2)
       )
      )
      B=&FN_Select_Float_from_Time_Of_Day(Morning=1, Day=1, Evening=1, Night=0.2)
     ))
    ScaleA=true
    ScaleRGB=true
   )
   &DynamicMaterialParameters(
    Index 0 Param 1=&SelectFloatFromArray(
     Direct Array Index=&MakeIntFromFloat(Float=&Multiply_Float(A=$Particles.NormalizedAge, B=20))
     Direct Array Mode=::NewEnumerator1
     Float Selection Array=DataInterface() {propertyValues {FloatData [0.04,0.31,0.67,0.91,-0.01,0.1,0.48,0,0.01,0.19]}}
    )
    Module.Param0WriteEnabled=true
    Module.Param1WriteEnabled=true
    Module.Param2WriteEnabled=true
    Module.Param3WriteEnabled=true
   ) {module=DynamicMaterialParameters001}
  ]
  rendererClasses [&NiagaraMeshRendererProperties]
  renderers [{rendererClass=&NiagaraMeshRendererProperties, rendererIndex=0}]
 }
 Smoke_Additional {
  emitterUpdateScript [
   &EmitterState(Life Cycle Mode=::NewEnumerator1, Loop Duration=1, UseLoopDelay=false)
   &SpawnRate(SpawnRate=3, Use Spawn Probability=false)
  ]
  particleSpawnScript [
   &InitializeParticle(
    Color=<rgba 0.01 0.01 0.01 1>
    Lifetime=&RandomRangeFloat(Minimum=2, Maximum=2.5)
    Mesh Scale=<1.5 1.5 4>
    Mesh Scale Mode=::NewEnumerator4
    Sprite Size=<20 40>
    Sprite Size Mode=::NewEnumerator4
    UsePositionOffset=false
    Write Material Random=true
   )
   &FN_SetSpawnGroup(Number of Spawn Groups=3)
   /SetVariables() {isSetParameters=true, module=SetVariables_#10}
   &SkeletalMeshLocation(
    Bone / Socket Index=&SelectIntFromArray(Direct Array Index=$Particles.SpawnGroup, Int Selection Array=DataInterface() {propertyValues {IntData [2,3,4]}})
    MaskSpawnGroup=false
    Mesh Sampling Type=::NewEnumerator5
    Orientation Sampling=::NewEnumerator1
    Skeletal Mesh=$System.Skeletal Mesh
    Socket Sampling Mode=::NewEnumerator5
   )
   /SetVariables(Particles.MeshOrientation=&MakeQuatFromAxisAngle(Axis=<0 0 1>, AngleInDegrees=&RandomRangeFloat(Minimum=0, Maximum=180), NormalizeAxis=true)) {isSetParameters=true, module=SetVariables_#11}
   &ApplyInitialForces()
  ]
  particleUpdateScript [
   &ParticleState()
   &ScaleMeshSize(Scale Factor=&MakeVector(X=0.8, Y=0.8, Z=&RandomRangeFloat(Minimum=1, Maximum=1.5)))
   &SkeletalMeshLocation(
    Bone / Socket Index=&SelectIntFromArray(Direct Array Index=$Particles.SpawnGroup, Int Selection Array=DataInterface() {propertyValues {IntData [2,3,4]}})
    MaskSpawnGroup=false
    Mesh Sampling Type=::NewEnumerator5
    Orientation Sampling=::NewEnumerator1
    Sampled Position Offset=<0 0 -1>
    Skeletal Mesh=$System.Skeletal Mesh
    Socket Sampling Mode=::NewEnumerator5
   ) {module=SkeletalMeshLocation003}
   &ScaleColor(
    Scale Alpha=&Multiply_Float(
     A=&FN_FrontEndFloatModifier(Gameplay=1, Front End=1)
     B=&FloatFromCurve(FloatCurve=DataInterface() {propertyValues {Curve {keys [
          <at 0 0 RCIM_Cubic>
          <at 0.7 1 RCIM_Cubic>
          {arriveTangent=0.5326, interpMode=RCIM_Cubic, leaveTangent=0.5326, tangentMode=RCTM_User, time=1, value=0}
         ]}}})
    )
    Scale RGB=&Multiply_VectorByFloat(
     Vector=&Multiply_VectorByFloat(Vector=<1 1 1>, Float=&FN_FrontEndFloatModifier(Gameplay=2, Front End=1))
     Float=&FNTimeOfDayScalar(Night Scale=0.3, Morning Scale=0.3, Evening Scale=0.4)
    )
    ScaleA=true
    ScaleRGB=false
   )
   &ScaleColor(
    Color Value To Scale=$StackContext.Color
    Scale Alpha=&Lerp_Float(A=1, B=1, Alpha=$System.KillReactBlend)
    ScaleA=true
    ScaleRGB=false
   ) {module=ScaleColor001}
   &InterpolateOverTime(
    Rate Of Change:NiagaraFloat=&RandomRangeFloat(Minimum=5, Maximum=15)
    Target Vector Value:Vector3f=&ConvertPositionToVector(Input Position=$Particles.Position, Convert To Absolute Large World Vector=false)
    Variable Type=::NewEnumerator2
   )
   /SetVariables(Particles.SpriteDynParams:Vector4f=&MakeVector4(
     X=0
     Y=&FloatFromCurve(
      FloatCurve=DataInterface() {propertyValues {Curve {keys [{tangentMode=RCTM_User, time=0, value=0}, {tangentMode=RCTM_User, time=1, value=0.5}]}}}
      Scale Curve=&FloatFromCurve(FloatCurve=DataInterface() {propertyValues {Curve {keys [<at 0 1.7 RCIM_Cubic>,<at 1 1.7 RCIM_Cubic>]}}})
     )
     Z=0
     W=0
    )) {isSetParameters=true, module=SetVariables_#12}
   /SetVariables(Particles.Interpolate Dynamic=&MakeFloat4FromVectorAndFloat("Vector (RGB)"=$StackContext.InterpolateOverTime.InterpolateOverTime_V3_.Moving Average, "Float (Alpha)"=1)) {isSetParameters=true, module=SetVariables_#13}
   /SetVariables(Particles.InterpolatedPosition:NiagaraPosition=$Particles.InterpolateOverTime.InterpolateOverTime_V3_.Moving Average) {isSetParameters=true, module=SetVariables_#14}
   &DynamicMaterialParameters(
    Index 0 Param 1=&RandomRangeFloat(Minimum=0.5, Maximum=0.3)
    Index 0 Param 2=&RandomRangeFloat(Minimum=0.2, Maximum=0.5)
    Index 0 Param 3=50
    Index 0 Param 4=1
    Index 1 Param 1=&MakeFloatFromVector4(Vector 4=$Particles.Interpolate Dynamic, Channel=::NewEnumerator0)
    Index 1 Param 2=&MakeFloatFromVector4(Vector 4=$Particles.Interpolate Dynamic, Channel=::NewEnumerator1)
    Index 1 Param 3=&Lerp_Float(
     A=&MakeFloatFromVector(Vector=Linked() {linkedVariable {name=Particles.Position, type=NiagaraPosition}}, Channel=::NewEnumerator2)
     B=&MakeFloatFromVector4(Vector 4=$Particles.Interpolate Dynamic, Channel=::NewEnumerator2)
     Alpha=0.5
    )
    Index 1 Param 4=1.5
    Index 2 Param 1=3
    Index 2 Param 2=0.1
    Module.Param0WriteEnabled=true
    Module.Param0WriteEnabled1=true
    Module.Param0WriteEnabled2=true
    Module.Param1WriteEnabled=true
    Module.Param1WriteEnabled1=true
    Module.Param1WriteEnabled2=true
    Module.Param2WriteEnabled=true
    Module.Param2WriteEnabled1=true
    Module.Param2WriteEnabled2=true
    Module.Param3WriteEnabled=true
    Module.Param3WriteEnabled1=true
    Module.Param3WriteEnabled2=true
    Write Parameter Index 1=true
    Write Parameter Index 2=true
   ) {module=DynamicMaterialParameters001}
   /SetVariables(Particles.SmokeStripColor:LinearColor=&MakeLinearColorFromVectorAndFloat(
     "Vector (RGB)"=<1 1 1>
     "Float (Alpha)"=&FloatFromCurve(
      FloatCurve=DataInterface() {propertyValues {Curve {keys [
          {interpMode=RCIM_Cubic, tangentMode=RCTM_User, time=0, value=0}
          {interpMode=RCIM_Cubic, tangentMode=RCTM_User, time=0.5, value=1}
          {tangentMode=RCTM_User, time=1, value=0}
         ]}}}
      Scale Curve=&Multiply_Float(
       A=&RandomRangeFloat(Minimum=0.06, Maximum=0.08)
       B=$System.InterpolateOverTime001.InterpolateOverTime_Float.Moving Average
      )
     )
    )) {isSetParameters=true, module=SetVariables_#15}
   &SpriteFacingAndAlignment(Sprite Alignment=<0 0 1>, WriteAlignment=true, WriteFacing=true)
  ]
  rendererClasses [&NiagaraMeshRendererProperties]
  renderers [{rendererClass=&NiagaraMeshRendererProperties, rendererIndex=0}]
 }
 Smoke_Main {
  emitterUpdateScript [
   &EmitterState(Life Cycle Mode=::NewEnumerator1, Loop Duration=1, UseLoopDelay=false)
   &SpawnRate(SpawnRate=5, Use Spawn Probability=false)
  ]
  particleSpawnScript [
   &InitializeParticle(
    Color=<rgba 0.01 0.01 0.01 1>
    Lifetime=&RandomRangeFloat(Minimum=1, Maximum=2.5)
    Mesh Scale=<1.5 1.5 4>
    Mesh Scale Mode=::NewEnumerator4
    Sprite Size=<20 40>
    Sprite Size Mode=::NewEnumerator4
    UsePositionOffset=false
    Write Material Random=true
   )
   &FN_SetSpawnGroup(Number of Spawn Groups=3)
   /SetVariables() {isSetParameters=true, module=SetVariables_#10}
   &CameraOffset(Camera Offset Amount=3)
   &SkeletalMeshLocation(
    Bone / Socket Index=&SelectIntFromArray(Direct Array Index=$Particles.SpawnGroup, Int Selection Array=DataInterface() {propertyValues {IntData [2,3,4]}})
    Bone Sampling Mode=::NewEnumerator4
    MaskSpawnGroup=false
    Orientation Sampling=::NewEnumerator1
    Skeletal Mesh=$System.Skeletal Mesh
   )
   /SetVariables(Particles.MeshOrientation=&MakeQuatFromAxisAngle(Axis=<0 0 1>, AngleInDegrees=&RandomRangeFloat(Minimum=-30, Maximum=30), NormalizeAxis=true)) {isSetParameters=true, module=SetVariables_#11}
   &ApplyInitialForces()
  ]
  particleUpdateScript [
   &ParticleState()
   &ScaleMeshSize(Scale Factor=&SelectVectorFromArray(
     Direct Array Index=$Particles.SpawnGroup
     Vector Selection Array=DataInterface() {propertyValues {FloatData [<0.4 0.4 0.6>, <0.5 0.5 0.5>, <0.35 0.35 0.35>]}}
    ))
   &SkeletalMeshLocation(
    Bone / Socket Index=&SelectIntFromArray(Direct Array Index=$Particles.SpawnGroup, Int Selection Array=DataInterface() {propertyValues {IntData [2,3,4]}})
    MaskSpawnGroup=false
    Mesh Sampling Type=::NewEnumerator5
    Orientation Sampling=::NewEnumerator1
    Skeletal Mesh=$System.Skeletal Mesh
    Socket Sampling Mode=::NewEnumerator5
   ) {module=SkeletalMeshLocation003}
   &ScaleColor(
    Scale Alpha=&Multiply_Float(
     A=&FN_FrontEndFloatModifier(Gameplay=1, Front End=1)
     B=&FloatFromCurve(FloatCurve=DataInterface() {propertyValues {Curve {keys [
          <at 0 0 RCIM_Cubic>
          <at 0.7 1 RCIM_Cubic>
          {arriveTangent=0.5326, interpMode=RCIM_Cubic, leaveTangent=0.5326, tangentMode=RCTM_User, time=1, value=0}
         ]}}})
    )
    Scale RGB=&Multiply_VectorByFloat(
     Vector=&Multiply_VectorByFloat(Vector=<1 1 1>, Float=&FN_FrontEndFloatModifier(Gameplay=2, Front End=1))
     Float=&FNTimeOfDayScalar(Night Scale=0.3, Morning Scale=0.3, Evening Scale=0.4)
    )
    ScaleA=true
    ScaleRGB=false
   )
   &ScaleColor(
    Color Value To Scale=$StackContext.Color
    Scale Alpha=&Lerp_Float(A=0.5, B=1, Alpha=$System.KillReactBlend)
    ScaleA=true
    ScaleRGB=false
   ) {module=ScaleColor001}
   &InterpolateOverTime(
    Rate Of Change:NiagaraFloat=&RandomRangeFloat(Minimum=5, Maximum=15)
    Target Vector Value:Vector3f=&ConvertPositionToVector(Input Position=$Particles.Position, Convert To Absolute Large World Vector=false)
    Variable Type=::NewEnumerator2
   )
   /SetVariables(Particles.SpriteDynParams:Vector4f=&MakeVector4(
     X=0
     Y=&FloatFromCurve(
      FloatCurve=DataInterface() {propertyValues {Curve {keys [{tangentMode=RCTM_User, time=0, value=0}, {tangentMode=RCTM_User, time=1, value=0.5}]}}}
      Scale Curve=&FloatFromCurve(FloatCurve=DataInterface() {propertyValues {Curve {keys [<at 0 1.7 RCIM_Cubic>,<at 1 1.7 RCIM_Cubic>]}}})
     )
     Z=0
     W=0
    )) {isSetParameters=true, module=SetVariables_#12}
   /SetVariables(Particles.Interpolate Dynamic=&MakeFloat4FromVectorAndFloat("Vector (RGB)"=$StackContext.InterpolateOverTime.InterpolateOverTime_V3_.Moving Average, "Float (Alpha)"=1)) {isSetParameters=true, module=SetVariables_#13}
   /SetVariables(Particles.InterpolatedPosition:NiagaraPosition=$Particles.InterpolateOverTime.InterpolateOverTime_V3_.Moving Average) {isSetParameters=true, module=SetVariables_#14}
   &DynamicMaterialParameters(
    Index 0 Param 1=&RandomRangeFloat(Minimum=0.9, Maximum=0.75)
    Index 0 Param 2=&RandomRangeFloat(Minimum=0.58, Maximum=1)
    Index 0 Param 3=50
    Index 0 Param 4=1
    Index 1 Param 1=&MakeFloatFromVector4(Vector 4=$Particles.Interpolate Dynamic, Channel=::NewEnumerator0)
    Index 1 Param 2=&MakeFloatFromVector4(Vector 4=$Particles.Interpolate Dynamic, Channel=::NewEnumerator1)
    Index 1 Param 3=&Lerp_Float(
     A=&MakeFloatFromVector(Vector=Linked() {linkedVariable {name=Particles.Position, type=NiagaraPosition}}, Channel=::NewEnumerator2)
     B=&MakeFloatFromVector4(Vector 4=$Particles.Interpolate Dynamic, Channel=::NewEnumerator2)
     Alpha=&MakeCustomFloatFromBool(Bool=$User.IsSkydiving, True Float=0.1, False Float=0.35)
    )
    Index 1 Param 4=1.5
    Index 2 Param 1=&Lerp_Float(A=6, B=8, Alpha=$System.KillReactBlend)
    Index 2 Param 2=&Lerp_Float(A=0.2, B=0.45, Alpha=$System.KillReactBlend)
    Module.Param0WriteEnabled=true
    Module.Param0WriteEnabled1=true
    Module.Param0WriteEnabled2=true
    Module.Param1WriteEnabled=true
    Module.Param1WriteEnabled1=true
    Module.Param1WriteEnabled2=true
    Module.Param2WriteEnabled=true
    Module.Param2WriteEnabled1=true
    Module.Param2WriteEnabled2=true
    Module.Param3WriteEnabled=true
    Module.Param3WriteEnabled1=true
    Module.Param3WriteEnabled2=true
    Write Parameter Index 1=true
    Write Parameter Index 2=true
   ) {module=DynamicMaterialParameters001}
   /SetVariables(Particles.SmokeStripColor:LinearColor=&MakeLinearColorFromVectorAndFloat(
     "Vector (RGB)"=<1 1 1>
     "Float (Alpha)"=&FloatFromCurve(
      FloatCurve=DataInterface() {propertyValues {Curve {keys [
          {interpMode=RCIM_Cubic, tangentMode=RCTM_User, time=0, value=0}
          {interpMode=RCIM_Cubic, tangentMode=RCTM_User, time=0.5, value=1}
          {tangentMode=RCTM_User, time=1, value=0}
         ]}}}
      Scale Curve=&Multiply_Float(A=0.2, B=$System.InterpolateOverTime001.InterpolateOverTime_Float.Moving Average)
     )
    )) {isSetParameters=true, module=SetVariables_#15}
   /SetVariables(Particles.New Vector 2D:Vector2f=<2 2>) {isSetParameters=true}
  ]
  rendererClasses [&NiagaraMeshRendererProperties]
  renderers [{rendererClass=&NiagaraMeshRendererProperties, rendererIndex=0}]
 }
}
summary {
 systemName=NS_Backpack_HavyRoar
 emitters [
  {emitterName=OuterDisc_L, rendererClasses [&NiagaraMeshRendererProperties]}
  {emitterName=OuterDisc_R, rendererClasses [&NiagaraMeshRendererProperties]}
  {emitterName=Smoke_Additional, rendererClasses [&NiagaraMeshRendererProperties]}
  {emitterName=Light, rendererClasses [&NiagaraLightRendererProperties]}
  {emitterName=Slimshady_Overlay, rendererClasses [&NiagaraMeshRendererProperties]}
  {emitterName=Smoke_Main, rendererClasses [&NiagaraMeshRendererProperties]}
 ]
 userVariables [
  User.bIsFrontEnd {type=NiagaraBool}
  User.bIsFrontEndPreview {type=NiagaraBool}
  User.IsSkydiving {type=NiagaraBool}
 ]
}
