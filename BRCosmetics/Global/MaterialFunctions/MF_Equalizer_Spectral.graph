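# Material function that draws an audio-reactive spectrum equalizer as a grid of cells, outputting colored value, transformed UVs and UV mask.
# Offset, rotated, scaled UVs are quantized into rows and columns, with responsiveness controlling per-bar variation in sampled audio MPC history.
# Custom HLSL indexes four spectrum bands and optional per-band colors, with switchable gradient blending and amplitude scaling.
# Optional layered-sine ambient motion animates the bars and fades when music starts.
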MaterialFunction /BRCosmetics/Global/MaterialFunctions/MF_Equalizer_Spectral
graph {nodes [
  n1 = VectorParameter() {group=Equalizer, parameterName=EQColor_Bass, sortPriority=70, defaultValue=<rgba 1 0 1 1>}
  n2 = NamedRerouteUsage()
  MaterialExpressionStaticSwitchParameter_3 = StaticSwitchParameter(
   True=LinearInterpolate(
    A=VectorParameter() {group=Equalizer, parameterName=EQColor_Bass_Blend, sortPriority=71, defaultValue=<rgba 1 1 1 1>}.RGB
    B=@n1.RGB
    Alpha=@n2
   )
   False=@n1.RGB
  ) {group=Equalizer, parameterName=EQUseBlendedColor, sortPriority=0}
  n3 = VectorParameter() {group=Equalizer, parameterName=EQColor_LowerMid_Blend, sortPriority=81, defaultValue=<rgba 1 1 1 1>}
  n4 = VectorParameter() {group=Equalizer, parameterName=EQColor_LowerMid, sortPriority=80, defaultValue=<rgba 1 0 0 1>}
  MaterialExpressionStaticSwitchParameter_4 = StaticSwitchParameter(True=LinearInterpolate(A=@n3.RGB, B=@n4.RGB, Alpha=@n2), False=@n4.RGB) {group=Equalizer, parameterName=EQUseBlendedColor, sortPriority=0}
  n5 = VectorParameter() {group=Equalizer, parameterName=EQColor_UpperMid_Blend, sortPriority=91, defaultValue=<rgba 1 1 1 1>}
  n6 = VectorParameter() {group=Equalizer, parameterName=EQColor_UpperMid, sortPriority=90, defaultValue=<rgba 0 1 0 1>}
  MaterialExpressionStaticSwitchParameter_5 = StaticSwitchParameter(True=LinearInterpolate(A=@n5.RGB, B=@n6.RGB, Alpha=@n2), False=@n6.RGB) {group=Equalizer, parameterName=EQUseBlendedColor, sortPriority=0}
  n7 = VectorParameter() {group=Equalizer, parameterName=EQColor_Treble_Blend, sortPriority=101, defaultValue=<rgba 1 1 1 1>}
  n8 = VectorParameter() {group=Equalizer, parameterName=EQColor_Treble, sortPriority=100, defaultValue=<rgba 0 0 1 1>}
  MaterialExpressionStaticSwitchParameter_6 = StaticSwitchParameter(True=LinearInterpolate(A=@n7.RGB, B=@n8.RGB, Alpha=@n2), False=@n8.RGB) {group=Equalizer, parameterName=EQUseBlendedColor, sortPriority=0}
  MaterialExpressionNamedRerouteUsage_5 = NamedRerouteUsage()
  n9 = Custom!(
   in1=@MaterialExpressionStaticSwitchParameter_3
   in2=@MaterialExpressionStaticSwitchParameter_4
   in3=@MaterialExpressionStaticSwitchParameter_5
   in4=@MaterialExpressionStaticSwitchParameter_6
   alpha=@MaterialExpressionNamedRerouteUsage_5
  ) {
   code="##pragma once\r\n\r\nconst int numVals = 4;\r\n\r\nconst float3 values[numVals] = {\r\n\t\tin1, \r\n        in2,\r\n\t\tin3, \r\n        in4\r\n\t};\r\n\r\nconst uint a = uint(alpha);\r\n\r\nreturn values[a];"
   inputs [
    {inputName=in1, input {expression=&self:MaterialExpressionStaticSwitchParameter_3}}
    {inputName=in2, input {expression=&self:MaterialExpressionStaticSwitchParameter_4}}
    {inputName=in3, input {expression=&self:MaterialExpressionStaticSwitchParameter_5}}
    {inputName=in4, input {expression=&self:MaterialExpressionStaticSwitchParameter_6}}
    {inputName=alpha, input {expression=&self:MaterialExpressionNamedRerouteUsage_5}}
   ]
  }
  n10 = VectorParameter() {group=Equalizer, parameterName=EQColor, sortPriority=0, defaultValue=<rgba 1 1 1 1>}
  n11 = StaticSwitchParameter(
   True=LinearInterpolate(
    A=@n10.RGB
    B=VectorParameter() {group=Equalizer, parameterName=EQColor_Blend, sortPriority=0, defaultValue=<rgba 1 1 1 1>}.RGB
    Alpha=NamedRerouteUsage()
   )
   False=@n10.RGB
  ) {group=Equalizer, parameterName=EQUseBlendedColor, sortPriority=0}
  n12 = CollectionParameter() {collection=/Game/Athena/Events/AudioAnalysis/MPC_Audio_Reactive_Parameters, parameterName=Amplitude_Music}
  n13 = Saturate(None=NamedRerouteUsage())
  MaterialExpressionComponentMask_4 = ComponentMask(None=@n13) {r=true}
  MaterialExpressionComponentMask_5 = ComponentMask(None=@n13) {g=true}
  MaterialExpressionComponentMask_6 = ComponentMask(None=@n13) {b=true}
  MaterialExpressionComponentMask_7 = ComponentMask(None=@n13) {a=true}
  MaterialExpressionNamedRerouteUsage_6 = NamedRerouteUsage()
  n14 = Custom(
   in1=@MaterialExpressionComponentMask_4
   in2=@MaterialExpressionComponentMask_5
   in3=@MaterialExpressionComponentMask_6
   in4=@MaterialExpressionComponentMask_7
   alpha=@MaterialExpressionNamedRerouteUsage_6
  ) {
   code="##pragma once\r\n\r\nconst int numVals = 4;\r\n\r\n const float values[numVals] = {\r\n\t\tin1, \r\n        in2,\r\n\t\tin3, \r\n        in4\r\n\t};\r\nconst uint a = uint(alpha);\r\n\r\n\r\nreturn values[a];"
   outputType=CMOT_Float1
   inputs [
    {inputName=in1, input {expression=&self:MaterialExpressionComponentMask_4}}
    {inputName=in2, input {expression=&self:MaterialExpressionComponentMask_5}}
    {inputName=in3, input {expression=&self:MaterialExpressionComponentMask_6}}
    {inputName=in4, input {expression=&self:MaterialExpressionComponentMask_7}}
    {inputName=alpha, input {expression=&self:MaterialExpressionNamedRerouteUsage_6}}
   ]
  }
  n15 = Power(Base=@n14, Exp=ScalarParameter() {defaultValue=1.3333, group=Equalizer, parameterName=EQPower}) {constExponent=1.3333}
  n16 = ChannelMaskParameter(None=NamedRerouteUsage()) {maskChannel=Green, parameterName=EQLevels, defaultValue=<rgba 0 1 0 0>}
  n17 = ScalarParameter() {defaultValue=0.1, group=Equalizer, parameterName=EQBias}
  n18 = Subtract(
   A=StaticSwitchParameter(True=Multiply(A=@n12, B=@n15), False=@n15) {group=Equalizer, parameterName=EQMultiplyByAmplitude}
   B=OneMinus(None=Add(A=@n16.R, B=@n17) {constB=0.08, desc="pivot (increase to raise bottom"})
  )
  n19 = Multiply(A=ComponentMask(None=NamedRerouteUsage()) {r=true}) {constB=1.4}
  n20 = Multiply(A=@n19) {constB=0.064}
  n21 = Add(A=Multiply(A=Time(), B=ScalarParameter() {defaultValue=0.6667, group=Equalizer, parameterName=EQAmbientSpeed}), B=NamedRerouteUsage())
  n22 = Multiply(A=@n21) {constB=0.6666}
  n23 = ConstantBiasScale(None=Sine(None=Add(A=@n20, B=@n22)))
  n24 = Multiply(A=@n19) {constB=0.9}
  n25 = Multiply(A=@n21)
  n26 = ConstantBiasScale(None=Sine(None=Add(A=@n24, B=@n25)))
  n27 = Multiply(A=Power(Base=Add(A=@n23, B=@n26)) {constExponent=0.5, desc=higher more often}) {constB=0.8}
  n28 = Multiply(A=Power(Base=Saturate(None=@n27)), B=ScalarParameter() {defaultValue=0.25, group=Equalizer, parameterName=EQAmbientHeight}) {constB=0.6666, desc=ambient}
  n29 = ScalarParameter() {defaultValue=0.1, group=Equalizer, parameterName=EQAmbientHeightDuringReactive, sliderMax=1}
  n30 = OneMinus(None=Multiply(A=NamedRerouteUsage(), B=OneMinus(None=@n29)) {constB=0.9})
  n31 = StaticSwitchParameter(True=Add(A=@n18, B=Multiply(A=@n28, B=@n30) {desc=reduce ambient when music starts}), False=@n18) {defaultValue=true, group=Equalizer, parameterName=EQUseAmbient, sortPriority=0}
  n32 = Multiply(A=NamedRerouteUsage(), B=Multiply(A=Sign(None=@n31), B=NamedRerouteUsage()))
  FunctionOutput(None=Multiply(A=StaticSwitchParameter(True=@n9, False=@n11) {group=Equalizer, parameterName=EQUseColorPerBand, sortPriority=0}, B=Saturate(None=@n32))) {outputName=Value}
  n33 = NamedRerouteUsage()
  n34 = VectorParameter() {group=Equalizer, parameterName="EQPosition(RG)Scale(BA)", sortPriority=0, defaultValue=<rgba 0 0 1 1>}
  n35 = AppendVector(A=@n34.R, B=@n34.R)
  n36 = MaterialFunctionCall(
   UVs=MaterialFunctionCall(UVs=Add(A=@n33, B=@n35), "Rotation Angle (0-1)"=ScalarParameter() {group=Equalizer, parameterName=EQRotation, sortPriority=0}) {materialFunction=&CustomRotator}.Rotated Values
   Texture Scale=AppendVector(A=@n34.B, B=@n34.B)
  ) {materialFunction=&ScaleUVsByCenter}
  FunctionOutput(None=NamedRerouteDeclaration(None=@n36.UVs) {name=TransformedUVs, nodeColor=<rgba 0.4066 0 1 1>}) {outputName=EQUVs, sortPriority=1}
  FunctionOutput(None=NamedRerouteDeclaration(None=@n36.0-1 mask) {name=UVMask_0-1, nodeColor=<rgba 1 0 0.2639 1>}) {outputName=EQUVMask, sortPriority=2}
  n37 = MaterialFunctionCall(Float2=NamedRerouteUsage()) {materialFunction=&BreakOutFloat2Components}
  n38 = ScalarParameter() {defaultValue=16, group=Equalizer, parameterName=EQRows, sortPriority=10}
  n39 = MaterialFunctionCall(
   X=Multiply(A=@n37.R, B=ScalarParameter() {defaultValue=16, group=Equalizer, parameterName=EQColumns, sortPriority=10})
   Y=Multiply(A=@n37.G, B=@n38)
  ) {materialFunction=&MakeFloat2}
  NamedRerouteDeclaration(None=Frac(None=@n39.Result)) {name=CellUVCoords, nodeColor=<rgba 1 0.5391 0 1>}
  n40 = Multiply(A=NamedRerouteUsage()) {constB=2}
  n41 = MaterialFunctionCall(Float2=Abs(None=Subtract(A=@n40))) {materialFunction=&BreakOutFloat2Components}
  n42 = Constant()
  n43 = ScalarParameter() {defaultValue=0.9, group=Equalizer, parameterName=EQCellWidth, sortPriority=15}
  n44 = Constant()
  n45 = ScalarParameter() {defaultValue=0.9, group=Equalizer, parameterName=EQCellHeight, sortPriority=15}
  NamedRerouteDeclaration(None=Multiply(A=SphereMask(A=@n41.R, B=@n42, Radius=@n43) {attenuationRadius=0.4}, B=SphereMask(A=@n41.G, B=@n44, Radius=@n45) {attenuationRadius=0.4})) {name=CellsMask, nodeColor=<rgba 1 0 0.3281 1>}
  n46 = Floor(None=@n39.Result)
  NamedRerouteDeclaration(None=@n46) {name=QuantizedUVs, nodeColor=<rgba 1 0.3047 0 1>}
  NamedRerouteDeclaration(None=Divide(A=@n46, B=@n38)) {name=QuantizedUVsNormalized, nodeColor=<rgba 0 0.5312 1 1>}
  n47 = ComponentMask(None=NamedRerouteUsage()) {r=true}
  n48 = ScalarParameter() {
   defaultValue=0.95
   desc="0-1 | The higher this values is, the faster the EQ will  react but it will also have less variation between bars."
   group=Equalizer
   parameterName=EQResponsiveness
   sliderMax=1
  }
  n49 = OneMinus(None=Saturate(None=@n48))
  n50 = MaterialFunctionCall!(useSmoothStep, uvx=Saturate(None=Multiply(A=@n47, B=@n49) {constB=0.05})) {materialFunction=/Game/Accessories/FORT_Backpacks/Backpack_F_MED_Cactus_Rocker/Materials/SampleAudioMPCHistory}
  NamedRerouteDeclaration(None=@n50.Interpolated Amplitude Music) {name=amp, nodeColor=<rgba 0 0.2031 1 1>}
  n51 = NamedRerouteUsage()
  NamedRerouteDeclaration(None=Power(Base=ComponentMask(None=@n51) {g=true}, Exp=ScalarParameter() {defaultValue=1, group=Equalizer, parameterName=EQGradientPower}) {constExponent=0.6667}) {name=ramp, nodeColor=<rgba 0.5579 0 1 1>}
  n52 = ComponentMask(None=@n51) {r=true}
  NamedRerouteDeclaration(None=Floor(None=Multiply(A=@n52) {constB=4})) {name=index, nodeColor=<rgba 0 1 0.3672 1>}
  NamedRerouteDeclaration(None=@n50.Interpolated Spectrum) {name=spectrum, nodeColor=<rgba 1 0 0.9609 1>}
  NamedRerouteDeclaration(None=StaticSwitchParameter(True=TextureCoordinate() {coordinateIndex=1}, False=TextureCoordinate()) {group=Equalizer, parameterName=EQUseUVChannel1, sortPriority=0}) {name=BaseUVs, nodeColor=<rgba 1 0 1 1>}
 ]}
