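# Character material for the HeavyRoar backpack, adding an audio-reactive emissive display with a spectral equalizer and volume indicator.
# Standard character PreFX, SkinFX and PostFX functions supply the base attributes.
# Scaled UVs and music amplitude from a parameter collection form the volume bar, added to the equalizer output.
# Emissive is masked to the display UVs, desaturated while amplitude is positive, scaled by brightness and time of day, then fed through cinematic PostFX.
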Material /BRCosmetics/Accessories/FORT_Backpacks/Backpack_M_MED_HeavyRoar/Materials/M_MED_HeavyRoar_Backpack
graph {
 properties {
  bAutomaticallySetUsageInEditor=false
  bUseMaterialAttributes=true
  bUsedWithMeshDeformer=true
  bUsedWithMorphTargets=true
  bUsedWithSkeletalMesh=true
  pixelDepthOffsetMode=PDOM_Legacy
  refractionMethod=RM_IndexOfRefraction
  shadingModel=MSM_FromMaterialExpression
  subsurfaceProfile=&SS_HeroSkin_02
  usedShadingModels=Default Lit | Subsurface | Clear Coat | Subsurface Profile | Cloth | Eye
 }
 nodes [
  n1 = MaterialFunctionCall() {materialFunction=&BaseFN_CharacterFunction_PreFX}
  n2 = MaterialFunctionCall(
   AO=@n1.AO
   Distance Blend=@n1.AO
   Fresnel=@n1.AO
   Gmap=@n1.AO
   Pre Skinned Local Position=@n1.AO
   Material=MaterialFunctionCall(
    PreskinnedPosition=@n1.Pre skinned Local Position
    Gmap=@n1.Pre skinned Local Position
    MaterialAttributes=@n1.Pre skinned Local Position
    AO=@n1.Pre skinned Local Position
   ) {materialFunction=&CharacterSkinFX}.Result
  ) {materialFunction=&BaseFN_CharacterFunction_PostFX}
  n3 = MaterialFunctionCall() {materialFunction=/BRCosmetics/Global/MaterialFunctions/MF_Equalizer_Spectral}
  n4 = GetMaterialAttributes(None=@n2.Result) {attributeGetTypes [#1]}
  n5 = MaterialFunctionCall(UVs=Add(A=TextureCoordinate(), B=Constant2Vector() {g=-0.4, r=-0.11}), Texture Scale=Constant() {r=0.08}) {materialFunction=&ScaleUVsByCenter}
  n6 = Multiply(A=@n4.Emissive Color, B=OneMinus(None=Multiply(A=@n5.0-1 mask) {constB=0.9}))
  n7 = Multiply(A=ComponentMask(None=@n5.UVs) {r=true}, B=@n5.0-1 mask)
  n8 = CollectionParameter() {collection=/Game/Athena/Events/AudioAnalysis/MPC_Audio_Reactive_Parameters, parameterName=Amplitude_Music}
  n9 = Multiply(A=@n4.Emissive Color, B=Floor(None=Add(A=@n7, B=@n8) {constB=0.5}))
  n10 = CollectionParameter() {collection=&MPC_Audio_Reactive_Parameters, parameterName=Amplitude_Music}
  n11 = ComponentMask(None=@n3.EQUVs) {g=true}
  n12 = Ceil(None=Frac(None=Saturate(None=@n11)))
  n13 = Desaturation(None=Add(A=@n3.Value, B=Add(A=@n6, B=@n9) {desc=Volume Indicator}), Fraction=Multiply(A=Ceil(None=@n10), B=@n12))
  n14 = ScalarParameter() {defaultValue=2, group=Emissive, parameterName=DisplayBrightness}
  n15 = Constant() {r=0.8}
  n16 = Multiply(
   A=Add(A=Multiply(A=@n12, B=@n14))
   B=MaterialFunctionCall(Value_Morning=@n15, Value_Day=Constant() {r=1}, Value_Evening=@n15, Value_Night=Constant() {r=0.4}) {materialFunction=&Scalar_Time_of_Day}.Result
  )
  MaterialFunctionCall(In=SetMaterialAttributes(MaterialAttributes=@n2.Result, Emissive Color=Multiply(A=@n13, B=@n16)) {attributeSetTypes [#1]}) {materialFunction=&CINE_MF_CharPostFX}
 ]
}
