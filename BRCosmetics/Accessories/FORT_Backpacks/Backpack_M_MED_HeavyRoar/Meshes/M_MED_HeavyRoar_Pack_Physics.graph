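# Physics asset for the HeavyRoar backpack cosmetic mesh, providing simple collision bodies and a joint for the pack and its dyn_mic bone.
# Two capsule bodies cover the bones: a long, thin capsule along dyn_mic and a short, wider one on pack.
# A single constraint links dyn_mic to pack with narrow swing limits and a zero-degree twist limit, restricting the mic to slight swaying relative to the pack.
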PhysicsAsset /BRCosmetics/Accessories/FORT_Backpacks/Backpack_M_MED_HeavyRoar/Meshes/M_MED_HeavyRoar_Pack_Physics
bodies {
 dyn_mic [{shapeType=Capsule, length=10.137, radius=1.3786, shapeName=dyn_mic_capsule, center=<6.3633 0 0>, rotation=<rot 90 0 0>}]
 pack [{shapeType=Capsule, length=1.656, radius=2.9927, shapeName=pack_capsule, center=<-0.3732 -3.2022 -2.9676>, rotation=<rot 0 -90.4694 -89.4001>}]
}
constraints [{bone1Name=dyn_mic, bone2Name=pack, swing1LimitDegrees=10, swing2LimitDegrees=7, twistLimitDegrees=0}]
